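# Editor utility list-entry widget for a custom asset picker, showing one asset as a small thumbnail with its name and tooltip.
# When given a list object, it reads AssetPackageName and ThumbnailSettings by name via GetEditorProperty.
# It then resolves the package through the Asset Registry and applies the first result to the thumbnail and text.
# Mouse enter, leave and selection changes refresh a faint hover border tint and a selection highlight image.
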EditorUtilityWidgetBlueprint /Game/NeverCook/EditorUtilities/CustomAssetPicker/EUW_CustomAssetPickerItem
graphs [
 EventGraph {nodes [
   Event PreConstruct()
   > SetVisibility(self=Get imgHover() -> (imgHover:Image Object Reference), InVisibility=Hidden)
   n1 = Event OnListItemObjectSet()
   > HandleListItemObjectSet!(execute, self, listObject:Object Reference=@n1.ListItemObject) -> (then)
   n2 = Event On Item Selection Changed!() -> (OutputDelegate, then, bIsSelected:bool) {is=title K2Node_Event}
   > Set IsCurrentlySelectedInListView(IsCurrentlySelectedInListView:bool=@n2.bIsSelected) -> (Output_Get:bool)
   > ValidateSelectionVisuals()
   Event OnMouseEnter()
   > ValidateSelectionVisuals()
   Event OnMouseLeave!() -> (OutputDelegate, then, MouseEvent:"Pointer Event Structure (by ref)") {is=title K2Node_Event}
   > ValidateSelectionVisuals()
   Event Construct()
   Event Tick()
  ]}
 HandleListItemObjectSet {nodes [
   HandleListItemObjectSet!() -> (then, listObject:Object Reference) {is=title K2Node_FunctionEntry}
   > n1 = Sequence!(execute) -> (then_0, then_1, then_2:Exec)
   > n2 = GetEditorProperty!(execute:Exec, Object:Object Reference=Get listObject() -> (listObject:Object Reference), PropertyName:Name=AssetPackageName) -> (then:Exec, PropertyValue:String, ReturnValue:bool)
   > Branch(Condition=@n2.ReturnValue)
   > Set l_packageName(l_packageName:String=@n2) -> (Output_Get:String)
   n3 = GetEditorProperty(execute=@n1.then_1, Object=Get listObject(), PropertyName=ThumbnailSettings) -> (PropertyValue:Asset Thumbnail Widget Settings Structure)
   > Branch(Condition=@n3.ReturnValue)
   > Set l_thumbSettings(l_thumbSettings:Asset Thumbnail Widget Settings Structure=@n3) -> (Output_Get:Asset Thumbnail Widget Settings Structure)
   SetPackageName!(
    execute=@n1.then_2
    self
    PackageName:String=Get l_packageName() -> (l_packageName:String)
    InThumbnailSettings:"Asset Thumbnail Widget Settings Structure (by ref)"=Get l_thumbSettings() -> (l_thumbSettings:Asset Thumbnail Widget Settings Structure)
   ) -> (then)
  ]}
 SetPackageName {nodes [
   n1 = SetPackageName!() -> (then, PackageName:String, InThumbnailSettings:"Asset Thumbnail Widget Settings Structure (by ref)") {is=title K2Node_FunctionEntry}
   > n2 = GetAssetsByPackageName!(
    execute
    self:Asset Registry Interface=GetAssetRegistry()
    PackageName:Name=String To Name(InString=@n1)
    bIncludeOnlyOnDiskAssets:bool=false
    bSkipARFilteredAssets:bool=true
   ) -> (then, OutAssetData:Array of Asset Data Structures, ReturnValue:bool)
   > Branch(Condition=@n2.ReturnValue)
   n3 = "Get (a copy)"(Array:Array of Asset Data Structures=@n2) -> (Output:Asset Data Structure)
   > SetAsset!(
    execute
    self:Asset Thumbnail Widget Object Reference=Get AssetThumbnailWidget_0() -> (AssetThumbnailWidget_0:Asset Thumbnail Widget Object Reference)
    AssetData:"Asset Data Structure (by ref)"=@n3
   ) -> (then)
   n4 = Break Asset Data(AssetData=@n3)
   n5 = Get txtMain() -> (txtMain:Text Object Reference)
   > "SetText (Text)"(self=@n5, InText="To Text (Name)"!(InName:Name=@n4.AssetName) -> (ReturnValue:Text))
   > SetToolTipText!(execute, self:Widget Object Reference=@n5, InToolTipText:"Text (by ref)"="To Text (Name)"(InName=@n4)) -> (then)
   > SetThumbnailSettings!(
    execute
    self:Asset Thumbnail Widget Object Reference=Get AssetThumbnailWidget_0()
    InThumbnailSettings:"Asset Thumbnail Widget Settings Structure (by ref)"=@n1.InThumbnailSettings
   ) -> (then)
  ]}
 ValidateSelectionVisuals {nodes [
   ValidateSelectionVisuals() {is=title K2Node_FunctionEntry}
   > SetBrushColor(
    self=Get Border_0() -> (Border_0:Border Object Reference)
    InBrushColor=Select(Option 0:Linear Color Structure=(R=0,G=0,B=0,A=0), Option 1:Linear Color Structure=(R=1,G=1,B=1,A=0.1), Index=IsHovered()) -> (ReturnValue:Linear Color Structure)
   )
   > SetVisibility(
    self=Get imgHover()
    InVisibility=Select(
     Option 0:ESlateVisibility Enum=Hidden
     Option 1:ESlateVisibility Enum=Visible
     Index=Get IsCurrentlySelectedInListView() -> (IsCurrentlySelectedInListView:bool)
    ) -> (ReturnValue:ESlateVisibility Enum)
   )
  ]}
]
variables [
 Index {type=boolean}
 IsCurrentlySelectedInListView {type=boolean}
]
widgets {
 info {rootWidgetClass=&Border, widgetCount=10}
 widgets [
  Border_0 {widgetClassPath=&Border, bIsVariable=true, widget=Border_0, properties {brushColor=<rgba 1 1 1 0>, padding=<ltrb 0 0 0 0>}}
  VerticalBox_0 {
   widgetClassPath=&VerticalBox
   parent=Border_0
   widget=VerticalBox_0
   properties {clipping=ClipToBoundsAlways}
   slot {class=BorderSlot, properties {padding=<ltrb 0 0 0 0>}}
  }
  Overlay_0 {widgetClassPath=&Overlay, parent=VerticalBox_0, widget=Overlay_0, slot {class=VerticalBoxSlot}}
  imgHover {
   widgetClassPath=&Image
   bIsVariable=true
   parent=Overlay_0
   widget=imgHover
   properties {visibility=Hidden, colorAndOpacity=<rgba 0.2806 0.5878 1 0.1>}
   slot {class=OverlaySlot, properties {horizontalAlignment=HAlign_Fill}}
  }
  SizeBox_1 {
   widgetClassPath=&SizeBox
   parent=Overlay_0
   widget=SizeBox_1
   properties {bOverride_HeightOverride=true, heightOverride=32}
   slot {class=OverlaySlot, properties {horizontalAlignment=HAlign_Fill, verticalAlignment=VAlign_Fill}}
  }
  stack {widgetClassPath=&StackBox, parent=SizeBox_1, widget=stack, slot {class=SizeBoxSlot}}
  ScaleBox_0 {
   widgetClassPath=&ScaleBox
   parent=stack
   widget=ScaleBox_0
   properties {stretchDirection=DownOnly}
   slot {class=StackBoxSlot, properties {padding=<ltrb 3 3 3 3>}}
  }
  AssetThumbnailWidget_0 {
   widgetClassPath=/Script/UMGEditor.AssetThumbnailWidget
   bIsVariable=true
   parent=ScaleBox_0
   widget=AssetThumbnailWidget_0
   properties {
    resolution=<32 32>
    thumbnailSettings {
     bAllowAssetSpecificThumbnailOverlay=false
     bAllowHintText=true
     bAllowRealTimeOnHovered=true
     bForceGenericThumbnail=false
     bOverrideAssetTypeColor=false
     colorStripOrientation=HorizontalBottomEdge
     genericThumbnailSize=24
     highlightedTextDelegate="(null).None"
     thumbnailLabel=ClassName
     assetTypeColorOverride=<rgba 0 0 0 0>
     hintColorAndOpacity=<rgba 0 0 0 0>
     padding=<ltrb 0 0 0 0>
    }
   }
   slot {class=ScaleBoxSlot, properties {horizontalAlignment=HAlign_Left}}
  }
  txtMain {
   widgetClassPath=&TextBlock
   bIsVariable=true
   parent=stack
   widget=txtMain
   properties {
    text=Text Block
    font {
     bForceMonospaced=false
     bMaterialIsStencil=false
     fontMaterial=null
     fontObject=&Roboto
     letterSpacing=0
     monospacedWidth=1
     size=9
     skewAmount=0
     typefaceFontName=Regular
     outlineSettings {
      bApplyOutlineToDropShadows=false
      bMiteredCorners=false
      bSeparateFillAlpha=false
      outlineMaterial=null
      outlineSize=0
      outlineColor=<rgba 0 0 0 1>
     }
    }
   }
   slot {class=StackBoxSlot, properties {verticalAlignment=VAlign_Center, padding=<ltrb 10 0 0 0>}}
  }
  Image_65 {
   widgetClassPath=&Image
   bIsVariable=true
   parent=VerticalBox_0
   widget=Image_65
   properties {brush {
     bIsDynamicallyLoaded=false
     drawAs=Image
     imageType=NoImage
     mirroring=NoMirror
     resourceName=None
     resourceObject=null
     tiling=NoTile
     imageSize=<32 2>
     margin=<ltrb 0 0 0 0>
     outlineSettings {
      bUseBrushTransparency=false
      roundingType=HalfHeightRadius
      width=0
      color {colorUseRule=UseColor_Specified, specifiedColor=<rgba 0 0 0 0>}
      cornerRadii=<0 0 0 0>
     }
     tintColor {colorUseRule=UseColor_Specified, specifiedColor=<rgba 0 0 0 1>}
     uVRegion {bIsValid=false, max=<0 0>, min=<0 0>}
    }}
   slot {class=VerticalBoxSlot}
  }
 ]
}
