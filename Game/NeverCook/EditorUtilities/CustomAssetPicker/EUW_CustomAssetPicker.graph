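# Reusable editor utility widget that lets users pick an asset from a labeled, searchable combo box.
# GatherAssets fills the options with Asset Registry package names, filtered by class, path, exact names, tag values and collection membership.
# The list refreshes on opening or via a refresh button, depending on ValidateListOnOpening, and tries to restore the previous selection.
# Selection changes broadcast the asset data, and a find button syncs the Content Browser to it.
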EditorUtilityWidgetBlueprint /Game/NeverCook/EditorUtilities/CustomAssetPicker/EUW_CustomAssetPicker
graphs [
 EventGraph {nodes [
   n1 = "On Selection Changed (SearchableStringComboBox_85)"!() -> (OutputDelegate, then, SelectedItem:String, SelectionType:ESelectInfo Enum) {is=title K2Node_ComponentBoundEvent}
   > n2 = GetSelectedAssetData!(execute, self:Self Object Reference) -> (then, AssetData:Asset Data Structure, HasSelection:bool)
   > Call On Selection Changed!(
    execute:Exec
    self:EUW Custom Asset Picker Object Reference
    PackageName:String=@n1.SelectedItem
    SelectionType:ESelectInfo Enum=@n1.SelectionType
    AssetData:Asset Data Structure=@n2
   ) -> (then:Exec) {is=title K2Node_CallDelegate}
   "On Opening (SearchableStringComboBox_85)"() {is=title K2Node_ComponentBoundEvent}
   > Branch(Condition=Get ValidateListOnOpening() -> (ValidateListOnOpening:bool))
   > RefreshList()
   "On Clicked (btnFind)"() {is=title K2Node_ComponentBoundEvent}
   > n3 = GetSelectedAssetData()
   > Branch(Condition=@n3.HasSelection)
   n4 = Break Asset Data(AssetData=@n3)
   > SyncBrowserToObjects(AssetPaths=Make Array("[0]":String="To String (Name)"(InName=@n4)) -> (Array:Array of Strings))
   "On Clicked (btnRefresh)"() {is=title K2Node_ComponentBoundEvent}
   > RefreshList()
   Event Construct()
   > n5 = Branch(Condition=Get ValidateListOnOpening())
   ValidateOptions(execute=@n5.else)
   Event PreConstruct()
   > ValidateLabel()
   > SetVisibility(
    self=Get btnRefresh() -> (btnRefresh:Editor Utility Button Object Reference)
    InVisibility=Select(Option 0=Visible, Option 1=Collapsed, Index=Get ValidateListOnOpening())
   )
  ]}
 ForceSetSelectedIndex {nodes [
   n1 = ForceSetSelectedIndex!() -> (then, Index:int) {is=title K2Node_FunctionEntry}
   n2 = GetOptionCount!(self:"Searchable Combo Box (String) Object Reference"=Get SearchableStringComboBox_85() -> (SearchableStringComboBox_85:"Searchable Combo Box (String) Object Reference")) -> (ReturnValue:int=0)
   > SetSelectedIndex!(
    execute
    self:"Searchable Combo Box (String) Object Reference"=Get SearchableStringComboBox_85()
    Index:int=SelectInt(A=int - int(A=@n2), B=0, bPickA="Equal (Integer)"(A=@n1))
   ) -> (then)
   > SetSelectedIndex(self=Get SearchableStringComboBox_85(), Index=@n1)
  ]}
 ForceSetSelectedOption {nodes [
   n1 = ForceSetSelectedOption!() -> (then, Option:String) {is=title K2Node_FunctionEntry}
   n2 = Get SearchableStringComboBox_85()
   n3 = GetOptionAtIndex!(self:"Searchable Combo Box (String) Object Reference"=@n2, Index:int=0) -> (ReturnValue:String)
   n4 = GetOptionCount(self=@n2) -> (ReturnValue=0)
   > SetSelectedIndex(self=Get SearchableStringComboBox_85(), Index=SelectInt(A=int - int(A=@n4), B=0, bPickA="Equal Exactly (String)"(A=@n3, B=@n1)))
   > SetSelectedOption!(execute, self:"Searchable Combo Box (String) Object Reference"=Get SearchableStringComboBox_85(), Option:String=@n1) -> (then)
   Get LastSelection() -> (LastSelection:String)
  ]}
 ForceSetSelection {nodes [
   ForceSetSelection() {is=title K2Node_FunctionEntry}
   > n1 = GetSelectedAssetData()
   > ForceSetSelectedOption!(execute, self, Option:String="To String (Name)"(InName=Break Asset Data(AssetData=@n1))) -> (then) {is=title K2Node_CallFunction}
  ]}
 GatherAssets {nodes [
   GatherAssets() {is=title K2Node_FunctionEntry}
   > n1 = Sequence!(execute) -> (then_0, then_1, then_2:Exec, then_3:Exec, then_4:Exec)
   > n2 = For Each Loop(Array:Array of Object Class References=Get ClassFilters() -> (ClassFilters:Array of Object Class References)) -> (Array Element:Object Class Reference)
   > Add(
    TargetArray:Array of Top Level Asset Path Structures=Get l_classPaths() -> (l_classPaths:Array of Top Level Asset Path Structures)
    NewItem:"Top Level Asset Path Structure (by ref)"="Get Class Path Name (Top Level Asset Path)"!(Class:Object Class Reference=@n2) -> (ReturnValue:Top Level Asset Path Structure) {is=title K2Node_CallFunction}
   ) {is=title K2Node_CallArrayFunction}
   n3 = GetAssets!(
    execute=@n1.then_1
    self:Asset Registry Interface=GetAssetRegistry()
    Filter:"ARFilter Structure (by ref)"=MakeARFilter!(
     PackageNames:Array of Names=Get ExactAssetPackageNames() -> (ExactAssetPackageNames:Array of Names)
     PackagePaths:Array of Names=Get PathFilters() -> (PathFilters:Array of Names)
     SoftObjectPaths:Array of Soft Object Path Structures
     ClassPaths:Array of Top Level Asset Path Structures=Get l_classPaths()
     RecursiveClassPathsExclusionSet:Set of Top Level Asset Path Structures
     ClassNames:Array of Names
     RecursiveClassesExclusionSet:Set of Names
     bRecursivePaths:bool=Get RecursivePaths() -> (RecursivePaths:bool)
     bRecursiveClasses:bool=Get RecursiveClasses() -> (RecursiveClasses:bool)
     bIncludeOnlyOnDiskAssets:bool=false
    ) -> (ReturnValue:ARFilter Structure)
    bSkipARFilteredAssets:bool=true
   ) -> (then, OutAssetData:Array of Asset Data Structures, ReturnValue:bool)
   > n4 = For Each Loop(Array:Array of Asset Data Structures=@n3) -> (Array Element:Asset Data Structure)
   > n5 = RunTagFilter!(execute, self, InAssetData:"Asset Data Structure (by ref)"=@n4) -> (then, ReturnValue:bool)
   > Branch(Condition=@n5)
   > Add!(
    execute
    TargetSet:Set of Names=Get l_packageNames() -> (l_packageNames:Set of Names)
    NewItem:"Name (by ref)"=Break Asset Data(AssetData=@n4)
   ) -> (then) {is=title K2Node_CallFunction}
   Branch(
    execute=@n1.then_2
    Condition=Is Valid Index!(TargetArray:Array of Names=Get CollectionFilters() -> (CollectionFilters:Array of Names), IndexToTest:int) -> (ReturnValue:bool)
   )
   > n6 = For Each Loop(Array:Array of Names=Get CollectionFilters()) -> (Array Element:Name)
   n7 = Intersection!(execute=@n6.Completed, A:Set of Names=Get l_collectionAssets() -> (l_collectionAssets:Set of Names), B:Set of Names=Get l_packageNames()) -> (then, Result:Set of Names)
   > Set l_packageNames(l_packageNames:Set of Names=@n7) -> (Output_Get:Set of Names)
   n8 = GetAssetsInCollection!(
    execute=@n6.LoopBody
    self:Asset Tags Subsystem Object Reference=Get AssetTagsSubsystem() -> (ReturnValue:Asset Tags Subsystem Object Reference) {is=title K2Node_GetEngineSubsystem}
    Name:Name=@n6
   ) -> (then, ReturnValue:Array of Asset Data Structures)
   > n9 = For Each Loop(Array:Array of Asset Data Structures=@n8) -> (Array Element:Asset Data Structure)
   > Add(TargetSet=Get l_collectionAssets(), NewItem=Break Asset Data(AssetData=@n9))
   n10 = To Array!(execute=@n1.then_4, A:Set of Names=Get l_packageNames()) -> (then, Result:Array of Names)
   > Return Node!(execute, PackageNames:Array of Names=@n10)
  ]}
 GetSelectedAssetData {nodes [
   GetSelectedAssetData() {is=title K2Node_FunctionEntry}
   n1 = Get SearchableStringComboBox_85()
   > n2 = GetAssetsByPackageName!(
    execute
    self:Asset Registry Interface=GetAssetRegistry()
    PackageName:Name=String To Name(InString=GetSelectedOption!(self:"Searchable Combo Box (String) Object Reference"=@n1) -> (ReturnValue:String))
    bIncludeOnlyOnDiskAssets:bool=false
    bSkipARFilteredAssets:bool=true
   ) -> (then, OutAssetData:Array of Asset Data Structures, ReturnValue:bool)
   > n3 = Branch(Condition=@n2.ReturnValue)
   > Return Node!(
    execute
    AssetData:Asset Data Structure="Get (a copy)"(Array:Array of Asset Data Structures=@n2) -> (Output:Asset Data Structure)
    HasSelection:bool=true
   )
   Return Node(execute=@n3.else, HasSelection=false)
  ]}
 LoadAndGetAssetObject {nodes [
   LoadAndGetAssetObject() {is=title K2Node_FunctionEntry}
   > n1 = GetSelectedAssetData() {is=title K2Node_CallFunction}
   > n2 = Branch(Condition=@n1.HasSelection)
   > Return Node!(execute, asset:Object Reference=GetAsset(InAssetData=@n1), HasSelection=true)
   Return Node(execute=@n2.else, HasSelection=false)
  ]}
 OnGenerateWidget {nodes [
   n1 = OnGenerateWidget!() -> (then, Item:String) {is=title K2Node_FunctionEntry}
   > n2 = Create Widget!(
    execute:Exec
    WorldContextObject:Object Reference
    Class:User Widget Class Reference=&/EUW_CustomAssetPickerItem.EUW_CustomAssetPickerItem_C
    OwningPlayer:Player Controller Object Reference
   ) -> (then:Exec, ReturnValue:EUW Custom Asset Picker Item Object Reference)
   > SetPackageName!(
    execute
    self:EUW Custom Asset Picker Item Object Reference=@n2
    PackageName:String=@n1
    InThumbnailSettings:"Asset Thumbnail Widget Settings Structure (by ref)"=Get ThumbnailSettings() -> (ThumbnailSettings:Asset Thumbnail Widget Settings Structure)
   ) -> (then)
   > Return Node!(execute, ReturnValue:Widget Object Reference=@n2)
  ]}
 OnSelectionChanged {nodes [OnSelectionChanged!() -> (then, PackageName:String, SelectionType:ESelectInfo Enum, AssetData:Asset Data Structure) {is=title K2Node_FunctionEntry}]}
 RefreshList {nodes [
   RefreshList() {is=title K2Node_FunctionEntry}
   > Set LastSelection(LastSelection:String=GetSelectedOption(self=Get SearchableStringComboBox_85())) -> (Output_Get:String)
   > ValidateOptions()
   > ResetLastSelection()
  ]}
 ResetLastSelection {nodes [
   ResetLastSelection() {is=title K2Node_FunctionEntry}
   n1 = Get SearchableStringComboBox_85()
   > Branch(Condition="integer < integer"(A=GetSelectedIndex!(self:"Searchable Combo Box (String) Object Reference"=@n1) -> (ReturnValue:int=0)))
   > ForceSetSelectedOption(Option=Get LastSelection())
  ]}
 RunTagFilter {nodes [
   n1 = RunTagFilter!() -> (then, InAssetData:"Asset Data Structure (by ref)") {is=title K2Node_FunctionEntry}
   n2 = Get TagFilters() -> (TagFilters:Map of Names to Strings)
   > n3 = Branch(Condition="integer > integer"(A=Length!(TargetMap:Map of Names to Strings=@n2) -> (ReturnValue:int=0) {is=title K2Node_CallFunction}))
   n4 = Get TagFilters()
   > n5 = Values!(execute, TargetMap:Map of Names to Strings=@n4) -> (then, Values:Array of Strings)
   > n6 = Keys!(execute, TargetMap:Map of Names to Strings=@n4) -> (then, Keys:Array of Names)
   > n7 = For Each Loop(Array:Array of Names=@n6) -> (Array Element:Name)
   Return Node(execute=@n7.Completed, ReturnValue:bool=false)
   n8 = GetTagValue(InAssetData=@n1, InTagName=@n7)
   > Branch(Condition=@n8.ReturnValue)
   n9 = "Get (a ref)"(Array:Array of Strings=@n5, Dimension 1=@n7.Array Index) -> (Output:"String (by ref)")
   > n10 = Branch(Condition=IsEmpty(InString=@n9))
   > Return Node(ReturnValue=true)
   Branch(execute=@n10.else, Condition="Equal Exactly (String)"(A=@n8, B=@n9))
   > Return Node(ReturnValue=true)
   Return Node(execute=@n3.else, ReturnValue=true)
  ]}
 SetPackageNames {nodes [
   n1 = SetPackageNames!() -> (then, Array:Array of Names) {is=title K2Node_FunctionEntry}
   n2 = Get SearchableStringComboBox_85()
   > ClearOptions(self:"Searchable Combo Box (String) Object Reference"=@n2)
   > ClearSelection(self:"Searchable Combo Box (String) Object Reference"=@n2)
   > n3 = For Each Loop(Array=@n1)
   > AddOption!(
    execute
    self:"Searchable Combo Box (String) Object Reference"=Get SearchableStringComboBox_85()
    Option:String="To String (Name)"(InName=@n3)
   ) -> (then)
  ]}
 SetSelectedAssetPackageName {nodes [
   n1 = SetSelectedAssetPackageName!() -> (then, PackageName:String) {is=title K2Node_FunctionEntry}
   > SetSelectedOption(self=Get SearchableStringComboBox_85(), Option=@n1)
  ]}
 Test {nodes [
   Test() {is=title K2Node_FunctionEntry}
   > n1 = GetAssetsByPath!(
    execute
    self:Asset Registry Interface=GetAssetRegistry()
    PackagePath:Name=/Game/Developers/brycelumpkin/ArtContent/
    bRecursive:bool=true
    bIncludeOnlyOnDiskAssets:bool=false
   ) -> (then, OutAssetData:Array of Asset Data Structures, ReturnValue:bool)
   > n2 = For Each Loop(Array:Array of Asset Data Structures=@n1) -> (Array Element:Asset Data Structure)
   > AddOption(self=Get SearchableStringComboBox_85(), Option="To String (Name)"(InName=Break Asset Data(AssetData=@n2)))
  ]}
 ValidateLabel {nodes [
   ValidateLabel() {is=title K2Node_FunctionEntry}
   n1 = Get txtLabel() -> (txtLabel:Text Object Reference)
   n2 = Get Label() -> (Label:Text)
   > "SetText (Text)"(self=@n1, InText=@n2)
   > SetVisibility(self=@n1, InVisibility=Select(Option 0=SelfHitTestInvisible, Option 1=Collapsed, Index=TextIsEmpty(InText=@n2)))
   n3 = SlotAsHorizontalBoxSlot!(Widget:Widget Object Reference=Get ScaleBox_0() -> (ScaleBox_0:Scale Box Object Reference)) -> (ReturnValue:Horizontal Box Slot Object Reference)
   > SetSize!(
    execute
    self:Horizontal Box Slot Object Reference=@n3
    InSize:Slate Child Size Structure=Get LabelFillType() -> (LabelFillType:Slate Child Size Structure)
   ) -> (then)
   > SetPadding!(execute, self:Horizontal Box Slot Object Reference=@n3, InPadding:Margin Structure=Get LabelPadding() -> (LabelPadding:Margin Structure)) -> (then)
   > n4 = Set members in Slate Font Info!(
    execute:Exec
    StructRef:"Slate Font Info Structure (by ref)"=Get Font!(self:Text Object Reference=Get txtLabel()) -> (Font:Slate Font Info Structure)
    Size:float=Get LabelTextSize() -> (LabelTextSize:float)
   ) -> (then:Exec, StructOut:"Slate Font Info Structure (by ref)") {is=title K2Node_SetFieldsInStruct}
   > SetFont(self=Get txtLabel(), InFontInfo=@n4)
   > Set Justification!(
    execute
    Justification:ETextJustify Enum=Get LabelJustification() -> (LabelJustification:ETextJustify Enum=Left)
    self:Text Layout Widget Object Reference=Get txtLabel()
   ) -> (then, Output_Get:ETextJustify Enum=Left)
   Make Slate Child Size!(Value:float=1, SizeRule:ESlateSizeRule Enum=Automatic) -> (SlateChildSize:Slate Child Size Structure) {is=title K2Node_MakeStruct}
  ]}
 ValidateOptions {nodes [
   ValidateOptions() {is=title K2Node_FunctionEntry}
   > n1 = GatherAssets!(execute, self) -> (then, PackageNames:Array of Names) {is=title K2Node_CallFunction}
   > SetPackageNames!(execute, self, Array:Array of Names=@n1) -> (then) {is=title K2Node_CallFunction}
  ]}
]
variables [
 ClassFilters {type="array<Class>", default []}
 CollectionFilters {type="array<string>", default []}
 ExactAssetPackageNames {type="array<string>", default []}
 Label {type=string, default=""}
 LabelFillType {type=SlateChildSize, default {sizeRule=Automatic, value=1}}
 LabelJustification {type=ETextJustify, default=Left}
 LabelPadding {type=Margin, default=<ltrb 0 0 10 0>}
 LabelTextSize {type=number, default=12}
 LastSelection {type=string, default=""}
 OnSelectionChanged {default=()}
 PathFilters {type="array<string>", default []}
 RecursiveClasses {type=boolean, default=true}
 RecursivePaths {type=boolean, default=true}
 TagFilters {type="map<string>", default {}}
 ThumbnailSettings {
  type=AssetThumbnailWidgetSettings
  default {
   bAllowAssetSpecificThumbnailOverlay=false
   bAllowHintText=true
   bAllowRealTimeOnHovered=false
   bForceGenericThumbnail=false
   bOverrideAssetTypeColor=false
   colorStripOrientation=HorizontalBottomEdge
   genericThumbnailSize=32
   highlightedTextDelegate="(null).None"
   thumbnailLabel=ClassName
   assetTypeColorOverride=<rgba 0 0 0 0>
   hintColorAndOpacity=<rgba 0 0 0 0>
   padding=<ltrb 0 0 0 0>
  }
 }
 TimerHandle_WaitForClose {type=TimerHandle}
 ValidateListOnOpening {type=boolean, default=true}
 WaitingForClose {type=boolean}
]
widgets {
 info {rootWidgetClass=&Border, widgetCount=10}
 widgets [
  Border_93 {widgetClassPath=&Border, widget=Border_93, properties {verticalAlignment=VAlign_Top, brushColor=<rgba 0.0176 0.0176 0.0176 1>}}
  SizeBox_1 {
   widgetClassPath=&SizeBox
   parent=Border_93
   widget=SizeBox_1
   properties {bOverride_HeightOverride=true, heightOverride=32}
   slot {class=BorderSlot, properties {verticalAlignment=VAlign_Top}}
  }
  HorizontalBox_0 {widgetClassPath=&HorizontalBox, parent=SizeBox_1, widget=HorizontalBox_0, slot {class=SizeBoxSlot}}
  ScaleBox_0 {
   widgetClassPath=&ScaleBox
   bIsVariable=true
   parent=HorizontalBox_0
   widget=ScaleBox_0
   properties {stretchDirection=DownOnly}
   slot {class=HorizontalBoxSlot, properties {padding=<ltrb 3 3 5 3>}}
  }
  txtLabel {
   widgetClassPath=&TextBlock
   bIsVariable=true
   parent=ScaleBox_0
   widget=txtLabel
   properties {
    text=Text Block
    colorAndOpacity {colorUseRule=UseColor_Specified, specifiedColor=<rgba 0.4452 0.4452 0.4452 1>}
    font {
     bForceMonospaced=false
     bMaterialIsStencil=false
     fontMaterial=null
     fontObject=&Roboto
     letterSpacing=0
     monospacedWidth=1
     size=9
     skewAmount=0
     typefaceFontName=Regular
     outlineSettings {
      bApplyOutlineToDropShadows=false
      bMiteredCorners=false
      bSeparateFillAlpha=false
      outlineMaterial=null
      outlineSize=0
      outlineColor=<rgba 0 0 0 1>
     }
    }
   }
   slot {class=ScaleBoxSlot, properties {horizontalAlignment=HAlign_Fill}}
  }
  SearchableStringComboBox_85 {
   widgetClassPath=/Script/FortniteEditor.SearchableStringComboBox
   bIsVariable=true
   parent=HorizontalBox_0
   widget=SearchableStringComboBox_85
   properties {
    foregroundColor {colorUseRule=UseColor_Specified, specifiedColor=<rgba 0.5271 0.5271 0.5271 1>}
    itemStyle {
     bUseParentRowBrush=false
     activeBrush {
      bIsDynamicallyLoaded=false
      drawAs=Image
      imageType=NoImage
      mirroring=NoMirror
      resourceName=None
      resourceObject=null
      tiling=NoTile
      imageSize=<0 0>
      margin=<ltrb 0 0 0 0>
      outlineSettings {
       bUseBrushTransparency=false
       roundingType=HalfHeightRadius
       width=0
       color {colorUseRule=UseColor_Specified, specifiedColor=<rgba 0 0 0 0>}
       cornerRadii=<0 0 0 0>
      }
      tintColor {colorUseRule=UseColor_Specified, specifiedColor=<rgba 0 0.162 0.7454 1>}
      uVRegion {bIsValid=false, max=<0 0>, min=<0 0>}
     }
     activeHighlightedBrush {
      bIsDynamicallyLoaded=false
      drawAs=Image
      imageType=NoImage
      mirroring=NoMirror
      resourceName=None
      resourceObject=null
      tiling=NoTile
      imageSize=<0 0>
      margin=<ltrb 0 0 0 0>
      outlineSettings {
       bUseBrushTransparency=false
       roundingType=HalfHeightRadius
       width=0
       color {colorUseRule=UseColor_Specified, specifiedColor=<rgba 0 0 0 0>}
       cornerRadii=<0 0 0 0>
      }
      tintColor {colorUseRule=UseColor_Specified, specifiedColor=<rgba 0.0252 0.0319 0.0423 1>}
      uVRegion {bIsValid=false, max=<0 0>, min=<0 0>}
     }
     activeHoveredBrush {
      bIsDynamicallyLoaded=false
      drawAs=Image
      imageType=NoImage
      mirroring=NoMirror
      resourceName=None
      resourceObject=null
      tiling=NoTile
      imageSize=<0 0>
      margin=<ltrb 0 0 0 0>
      outlineSettings {
       bUseBrushTransparency=false
       roundingType=HalfHeightRadius
       width=0
       color {colorUseRule=UseColor_Specified, specifiedColor=<rgba 0 0 0 0>}
       cornerRadii=<0 0 0 0>
      }
      tintColor {colorUseRule=UseColor_Specified, specifiedColor=<rgba 0 0.162 0.7454 1>}
      uVRegion {bIsValid=false, max=<0 0>, min=<0 0>}
     }
     dropIndicator_Above {
      bIsDynamicallyLoaded=false
      drawAs=Box
      imageType=FullColor
      mirroring=NoMirror
      resourceName=../../../Engine/Content/Slate/Common/DropZoneIndicator_Above.png
      resourceObject=null
      tiling=NoTile
      imageSize=<0 0>
      margin=<ltrb 0.625 0.625 0 0>
      outlineSettings {
       bUseBrushTransparency=false
       roundingType=HalfHeightRadius
       width=0
       color {colorUseRule=UseColor_Specified, specifiedColor=<rgba 0 0 0 0>}
       cornerRadii=<0 0 0 0>
      }
      tintColor {colorUseRule=UseColor_Specified, specifiedColor=<rgba 0.0091 0.3515 0.9647 1>}
      uVRegion {bIsValid=false, max=<0 0>, min=<0 0>}
     }
     dropIndicator_Below {
      bIsDynamicallyLoaded=false
      drawAs=Box
      imageType=FullColor
      mirroring=NoMirror
      resourceName=../../../Engine/Content/Slate/Common/DropZoneIndicator_Below.png
      resourceObject=null
      tiling=NoTile
      imageSize=<0 0>
      margin=<ltrb 0.625 0 0 0.625>
      outlineSettings {
       bUseBrushTransparency=false
       roundingType=HalfHeightRadius
       width=0
       color {colorUseRule=UseColor_Specified, specifiedColor=<rgba 0 0 0 0>}
       cornerRadii=<0 0 0 0>
      }
      tintColor {colorUseRule=UseColor_Specified, specifiedColor=<rgba 0.0091 0.3515 0.9647 1>}
      uVRegion {bIsValid=false, max=<0 0>, min=<0 0>}
     }
     dropIndicator_Onto {
      bIsDynamicallyLoaded=false
      drawAs=Box
      imageType=FullColor
      mirroring=NoMirror
      resourceName=../../../Engine/Content/Slate/Common/DropZoneIndicator_Onto.png
      resourceObject=null
      tiling=NoTile
      imageSize=<0 0>
      margin=<ltrb 0.25 0.25 0.25 0.25>
      outlineSettings {
       bUseBrushTransparency=false
       roundingType=HalfHeightRadius
       width=0
       color {colorUseRule=UseColor_Specified, specifiedColor=<rgba 0 0 0 0>}
       cornerRadii=<0 0 0 0>
      }
      tintColor {colorUseRule=UseColor_Specified, specifiedColor=<rgba 0.0091 0.3515 0.9647 1>}
      uVRegion {bIsValid=false, max=<0 0>, min=<0 0>}
     }
     evenRowBackgroundBrush {
      bIsDynamicallyLoaded=false
      drawAs=Image
      imageType=NoImage
      mirroring=NoMirror
      resourceName=None
      resourceObject=null
      tiling=NoTile
      imageSize=<0 0>
      margin=<ltrb 0 0 0 0>
      outlineSettings {
       bUseBrushTransparency=false
       roundingType=HalfHeightRadius
       width=0
       color {colorUseRule=UseColor_Specified, specifiedColor=<rgba 0 0 0 0>}
       cornerRadii=<0 0 0 0>
      }
      tintColor {colorUseRule=UseColor_Specified, specifiedColor=<rgba 0.0103 0.0103 0.0103 1>}
      uVRegion {bIsValid=false, max=<0 0>, min=<0 0>}
     }
     evenRowBackgroundHoveredBrush {
      bIsDynamicallyLoaded=false
      drawAs=Image
      imageType=NoImage
      mirroring=NoMirror
      resourceName=None
      resourceObject=null
      tiling=NoTile
      imageSize=<0 0>
      margin=<ltrb 0 0 0 0>
      outlineSettings {
       bUseBrushTransparency=false
       roundingType=HalfHeightRadius
       width=0
       color {colorUseRule=UseColor_Specified, specifiedColor=<rgba 0 0 0 0>}
       cornerRadii=<0 0 0 0>
      }
      tintColor {colorUseRule=UseColor_Specified, specifiedColor=<rgba 0.0176 0.0176 0.0176 1>}
      uVRegion {bIsValid=false, max=<0 0>, min=<0 0>}
     }
     inactiveBrush {
      bIsDynamicallyLoaded=false
      drawAs=Image
      imageType=NoImage
      mirroring=NoMirror
      resourceName=None
      resourceObject=null
      tiling=NoTile
      imageSize=<0 0>
      margin=<ltrb 0 0 0 0>
      outlineSettings {
       bUseBrushTransparency=false
       roundingType=HalfHeightRadius
       width=0
       color {colorUseRule=UseColor_Specified, specifiedColor=<rgba 0 0 0 0>}
       cornerRadii=<0 0 0 0>
      }
      tintColor {colorUseRule=UseColor_Specified, specifiedColor=<rgba 0.0513 0.0953 0.159 1>}
      uVRegion {bIsValid=false, max=<0 0>, min=<0 0>}
     }
     inactiveHighlightedBrush {
      bIsDynamicallyLoaded=false
      drawAs=Image
      imageType=NoImage
      mirroring=NoMirror
      resourceName=None
      resourceObject=null
      tiling=NoTile
      imageSize=<0 0>
      margin=<ltrb 0 0 0 0>
      outlineSettings {
       bUseBrushTransparency=false
       roundingType=HalfHeightRadius
       width=0
       color {colorUseRule=UseColor_Specified, specifiedColor=<rgba 0 0 0 0>}
       cornerRadii=<0 0 0 0>
      }
      tintColor {colorUseRule=UseColor_Specified, specifiedColor=<rgba 0.0252 0.0319 0.0423 1>}
      uVRegion {bIsValid=false, max=<0 0>, min=<0 0>}
     }
     inactiveHoveredBrush {
      bIsDynamicallyLoaded=false
      drawAs=Image
      imageType=NoImage
      mirroring=NoMirror
      resourceName=None
      resourceObject=null
      tiling=NoTile
      imageSize=<0 0>
      margin=<ltrb 0 0 0 0>
      outlineSettings {
       bUseBrushTransparency=false
       roundingType=HalfHeightRadius
       width=0
       color {colorUseRule=UseColor_Specified, specifiedColor=<rgba 0 0 0 0>}
       cornerRadii=<0 0 0 0>
      }
      tintColor {colorUseRule=UseColor_Specified, specifiedColor=<rgba 0.0513 0.0953 0.159 1>}
      uVRegion {bIsValid=false, max=<0 0>, min=<0 0>}
     }
     oddRowBackgroundBrush {
      bIsDynamicallyLoaded=false
      drawAs=Image
      imageType=NoImage
      mirroring=NoMirror
      resourceName=None
      resourceObject=null
      tiling=NoTile
      imageSize=<0 0>
      margin=<ltrb 0 0 0 0>
      outlineSettings {
       bUseBrushTransparency=false
       roundingType=HalfHeightRadius
       width=0
       color {colorUseRule=UseColor_Specified, specifiedColor=<rgba 0 0 0 0>}
       cornerRadii=<0 0 0 0>
      }
      tintColor {colorUseRule=UseColor_Specified, specifiedColor=<rgba 0.0103 0.0103 0.0103 1>}
      uVRegion {bIsValid=false, max=<0 0>, min=<0 0>}
     }
     oddRowBackgroundHoveredBrush {
      bIsDynamicallyLoaded=false
      drawAs=Image
      imageType=NoImage
      mirroring=NoMirror
      resourceName=None
      resourceObject=null
      tiling=NoTile
      imageSize=<0 0>
      margin=<ltrb 0 0 0 0>
      outlineSettings {
       bUseBrushTransparency=false
       roundingType=HalfHeightRadius
       width=0
       color {colorUseRule=UseColor_Specified, specifiedColor=<rgba 0 0 0 0>}
       cornerRadii=<0 0 0 0>
      }
      tintColor {colorUseRule=UseColor_Specified, specifiedColor=<rgba 0.0176 0.0176 0.0176 1>}
      uVRegion {bIsValid=false, max=<0 0>, min=<0 0>}
     }
     parentRowBackgroundBrush {
      bIsDynamicallyLoaded=false
      drawAs=Image
      imageType=NoImage
      mirroring=NoMirror
      resourceName=None
      resourceObject=null
      tiling=NoTile
      imageSize=<32 32>
      margin=<ltrb 0 0 0 0>
      outlineSettings {
       bUseBrushTransparency=false
       roundingType=HalfHeightRadius
       width=0
       color {colorUseRule=UseColor_Specified, specifiedColor=<rgba 0 0 0 0>}
       cornerRadii=<0 0 0 0>
      }
      tintColor {colorUseRule=UseColor_Specified, specifiedColor=<rgba 1 1 1 1>}
      uVRegion {bIsValid=false, max=<0 0>, min=<0 0>}
     }
     parentRowBackgroundHoveredBrush {
      bIsDynamicallyLoaded=false
      drawAs=Image
      imageType=NoImage
      mirroring=NoMirror
      resourceName=None
      resourceObject=null
      tiling=NoTile
      imageSize=<32 32>
      margin=<ltrb 0 0 0 0>
      outlineSettings {
       bUseBrushTransparency=false
       roundingType=HalfHeightRadius
       width=0
       color {colorUseRule=UseColor_Specified, specifiedColor=<rgba 0 0 0 0>}
       cornerRadii=<0 0 0 0>
      }
      tintColor {colorUseRule=UseColor_Specified, specifiedColor=<rgba 1 1 1 1>}
      uVRegion {bIsValid=false, max=<0 0>, min=<0 0>}
     }
     selectedTextColor {colorUseRule=UseColor_Specified, specifiedColor=<rgba 1 1 1 1>}
     selectorFocusedBrush {
      bIsDynamicallyLoaded=false
      drawAs=Border
      imageType=FullColor
      mirroring=NoMirror
      resourceName=../../../Engine/Content/Slate/Common/Selector.png
      resourceObject=null
      tiling=Both
      imageSize=<0 0>
      margin=<ltrb 0.25 0.25 0.25 0.25>
      outlineSettings {
       bUseBrushTransparency=false
       roundingType=HalfHeightRadius
       width=0
       color {colorUseRule=UseColor_Specified, specifiedColor=<rgba 0 0 0 0>}
       cornerRadii=<0 0 0 0>
      }
      tintColor {colorUseRule=UseColor_Specified, specifiedColor=<rgba 0.701 0.225 0.003 1>}
      uVRegion {bIsValid=false, max=<0 0>, min=<0 0>}
     }
     textColor {colorUseRule=UseColor_Specified, specifiedColor=<rgba 0.5271 0.5271 0.5271 1>}
    }
    widgetStyle {
     comboButtonStyle {
      downArrowAlign=VAlign_Center
      buttonStyle {
       clickedSlateSound {resourceObject=null}
       disabled {
        bIsDynamicallyLoaded=false
        drawAs=RoundedBox
        imageType=NoImage
        mirroring=NoMirror
        resourceName=None
        resourceObject=null
        tiling=NoTile
        imageSize=<0 0>
        margin=<ltrb 0 0 0 0>
        outlineSettings {
         bUseBrushTransparency=false
         roundingType=FixedRadius
         width=1
         color {colorUseRule=UseColor_ColorTable, specifiedColor=<rgba 0 0 0 0>}
         cornerRadii=<4 4 4 4>
        }
        tintColor {colorUseRule=UseColor_Specified, specifiedColor=<rgba 0.0395 0.0395 0.0395 1>}
        uVRegion {bIsValid=false, max=<0 0>, min=<0 0>}
       }
       disabledForeground {colorUseRule=UseColor_ColorTable, specifiedColor=<rgba 0 0 0 0>}
       hovered {
        bIsDynamicallyLoaded=false
        drawAs=RoundedBox
        imageType=NoImage
        mirroring=NoMirror
        resourceName=None
        resourceObject=null
        tiling=NoTile
        imageSize=<0 0>
        margin=<ltrb 0 0 0 0>
        outlineSettings {
         bUseBrushTransparency=false
         roundingType=FixedRadius
         width=1
         color {colorUseRule=UseColor_ColorTable, specifiedColor=<rgba 0 0 0 0>}
         cornerRadii=<4 4 4 4>
        }
        tintColor {colorUseRule=UseColor_Specified, specifiedColor=<rgba 0.0048 0.0048 0.0048 1>}
        uVRegion {bIsValid=false, max=<0 0>, min=<0 0>}
       }
       hoveredForeground {colorUseRule=UseColor_ColorTable, specifiedColor=<rgba 0 0 0 0>}
       hoveredSlateSound {resourceObject=null}
       normal {
        bIsDynamicallyLoaded=false
        drawAs=RoundedBox
        imageType=NoImage
        mirroring=NoMirror
        resourceName=None
        resourceObject=null
        tiling=NoTile
        imageSize=<0 0>
        margin=<ltrb 0 0 0 0>
        outlineSettings {
         bUseBrushTransparency=false
         roundingType=FixedRadius
         width=1
         color {colorUseRule=UseColor_ColorTable, specifiedColor=<rgba 0 0 0 0>}
         cornerRadii=<4 4 4 4>
        }
        tintColor {colorUseRule=UseColor_Specified, specifiedColor=<rgba 0.0048 0.0048 0.0048 1>}
        uVRegion {bIsValid=false, max=<0 0>, min=<0 0>}
       }
       normalForeground {colorUseRule=UseColor_ColorTable, specifiedColor=<rgba 0 0 0 0>}
       normalPadding=<ltrb 8 1 8 1>
       pressed {
        bIsDynamicallyLoaded=false
        drawAs=RoundedBox
        imageType=NoImage
        mirroring=NoMirror
        resourceName=None
        resourceObject=null
        tiling=NoTile
        imageSize=<0 0>
        margin=<ltrb 0 0 0 0>
        outlineSettings {
         bUseBrushTransparency=false
         roundingType=FixedRadius
         width=1
         color {colorUseRule=UseColor_ColorTable, specifiedColor=<rgba 0 0 0 0>}
         cornerRadii=<4 4 4 4>
        }
        tintColor {colorUseRule=UseColor_Specified, specifiedColor=<rgba 0.0048 0.0048 0.0048 1>}
        uVRegion {bIsValid=false, max=<0 0>, min=<0 0>}
       }
       pressedForeground {colorUseRule=UseColor_ColorTable, specifiedColor=<rgba 0 0 0 0>}
       pressedPadding=<ltrb 8 2 8 0>
       pressedSlateSound {resourceObject=null}
      }
      contentPadding=<ltrb 0 0 0 0>
      downArrowImage {
       bIsDynamicallyLoaded=false
       drawAs=Image
       imageType=Vector
       mirroring=NoMirror
       resourceName=../../../Engine/Content/Slate/Starship/CoreWidgets/ComboBox/wide-chevron-down.svg
       resourceObject=null
       tiling=NoTile
       imageSize=<16 16>
       margin=<ltrb 0 0 0 0>
       outlineSettings {
        bUseBrushTransparency=false
        roundingType=HalfHeightRadius
        width=0
        color {colorUseRule=UseColor_Specified, specifiedColor=<rgba 0 0 0 0>}
        cornerRadii=<0 0 0 0>
       }
       tintColor {colorUseRule=UseColor_Specified, specifiedColor=<rgba 1 1 1 1>}
       uVRegion {bIsValid=false, max=<0 0>, min=<0 0>}
      }
      downArrowPadding=<ltrb 2 2 2 2>
      menuBorderBrush {
       bIsDynamicallyLoaded=false
       drawAs=RoundedBox
       imageType=NoImage
       mirroring=NoMirror
       resourceName=None
       resourceObject=null
       tiling=NoTile
       imageSize=<0 0>
       margin=<ltrb 0 0 0 0>
       outlineSettings {
        bUseBrushTransparency=false
        roundingType=FixedRadius
        width=1
        color {colorUseRule=UseColor_ColorTable, specifiedColor=<rgba 0 0 0 0>}
        cornerRadii=<0 0 0 0>
       }
       tintColor {colorUseRule=UseColor_Specified, specifiedColor=<rgba 0.0395 0.0395 0.0395 1>}
       uVRegion {bIsValid=false, max=<0 0>, min=<0 0>}
      }
      menuBorderPadding=<ltrb 0 0 0 0>
      shadowColorAndOpacity=<rgba 0 0 0 1>
      shadowOffset=<0 0>
     }
     contentPadding=<ltrb 0 0 0 0>
     menuRowPadding=<ltrb 8 3 8 3>
     pressedSlateSound {resourceObject=null}
     selectionChangeSlateSound {resourceObject=null}
    }
   }
   slot {class=HorizontalBoxSlot, properties {size {sizeRule=Fill, value=1}}}
  }
  btnFind {
   widgetClassPath=&EditorUtilityButton
   parent=HorizontalBox_0
   widget=btnFind
   properties {
    toolTipText=Find the selected asset in the content browser
    widgetStyle {
     clickedSlateSound {resourceObject=null}
     disabled {
      bIsDynamicallyLoaded=false
      drawAs=RoundedBox
      imageType=NoImage
      mirroring=NoMirror
      resourceName=None
      resourceObject=null
      tiling=NoTile
      imageSize=<0 0>
      margin=<ltrb 0 0 0 0>
      outlineSettings {
       bUseBrushTransparency=false
       roundingType=FixedRadius
       width=1
       color {colorUseRule=UseColor_ColorTable, specifiedColor=<rgba 0 0 0 0>}
       cornerRadii=<4 4 4 4>
      }
      tintColor {colorUseRule=UseColor_Specified, specifiedColor=<rgba 0.0395 0.0395 0.0395 1>}
      uVRegion {bIsValid=false, max=<0 0>, min=<0 0>}
     }
     disabledForeground {colorUseRule=UseColor_ColorTable, specifiedColor=<rgba 0 0 0 0>}
     hovered {
      bIsDynamicallyLoaded=false
      drawAs=RoundedBox
      imageType=NoImage
      mirroring=NoMirror
      resourceName=None
      resourceObject=null
      tiling=NoTile
      imageSize=<32 32>
      margin=<ltrb 0 0 0 0>
      outlineSettings {
       bUseBrushTransparency=false
       roundingType=FixedRadius
       width=1
       color {colorUseRule=UseColor_ColorTable, specifiedColor=<rgba 0 0 0 0>}
       cornerRadii=<4 4 4 4>
      }
      tintColor {colorUseRule=UseColor_Specified, specifiedColor=<rgba 0.0953 0.0953 0.0953 1>}
      uVRegion {bIsValid=false, max=<0 0>, min=<0 0>}
     }
     hoveredForeground {colorUseRule=UseColor_ColorTable, specifiedColor=<rgba 0 0 0 0>}
     hoveredSlateSound {resourceObject=null}
     normal {
      bIsDynamicallyLoaded=false
      drawAs=RoundedBox
      imageType=NoImage
      mirroring=NoMirror
      resourceName=None
      resourceObject=null
      tiling=NoTile
      imageSize=<32 32>
      margin=<ltrb 0 0 0 0>
      outlineSettings {
       bUseBrushTransparency=false
       roundingType=FixedRadius
       width=1
       color {colorUseRule=UseColor_ColorTable, specifiedColor=<rgba 0 0 0 0>}
       cornerRadii=<4 4 4 4>
      }
      tintColor {colorUseRule=UseColor_Specified, specifiedColor=<rgba 0.0395 0.0395 0.0395 1>}
      uVRegion {bIsValid=false, max=<0 0>, min=<0 0>}
     }
     normalForeground {colorUseRule=UseColor_ColorTable, specifiedColor=<rgba 0 0 0 0>}
     normalPadding=<ltrb 0 2 0 2>
     pressed {
      bIsDynamicallyLoaded=false
      drawAs=RoundedBox
      imageType=NoImage
      mirroring=NoMirror
      resourceName=None
      resourceObject=null
      tiling=NoTile
      imageSize=<32 32>
      margin=<ltrb 0 0 0 0>
      outlineSettings {
       bUseBrushTransparency=false
       roundingType=FixedRadius
       width=1
       color {colorUseRule=UseColor_ColorTable, specifiedColor=<rgba 0 0 0 0>}
       cornerRadii=<4 4 4 4>
      }
      tintColor {colorUseRule=UseColor_Specified, specifiedColor=<rgba 0.0284 0.0284 0.0284 1>}
      uVRegion {bIsValid=false, max=<0 0>, min=<0 0>}
     }
     pressedForeground {colorUseRule=UseColor_ColorTable, specifiedColor=<rgba 0 0 0 0>}
     pressedPadding=<ltrb 0 0 0 0>
     pressedSlateSound {resourceObject=null}
    }
   }
   slot {class=HorizontalBoxSlot, properties {horizontalAlignment=HAlign_Center, verticalAlignment=VAlign_Center, padding=<ltrb 3 0 3 0>}}
  }
  Image_75 {
   widgetClassPath=&Image
   bIsVariable=true
   parent=btnFind
   widget=Image_75
   properties {brush {
     bIsDynamicallyLoaded=false
     drawAs=Image
     imageType=FullColor
     mirroring=NoMirror
     resourceName=None
     resourceObject=/Game/Athena/UI/Challenges/Art/ChallengeTypeIcons/UI_ChallengeIcon_Find
     tiling=NoTile
     imageSize=<18 18>
     margin=<ltrb 0 0 0 0>
     outlineSettings {
      bUseBrushTransparency=false
      roundingType=HalfHeightRadius
      width=0
      color {colorUseRule=UseColor_Specified, specifiedColor=<rgba 0 0 0 0>}
      cornerRadii=<0 0 0 0>
     }
     tintColor {colorUseRule=UseColor_Specified, specifiedColor=<rgba 1 1 1 1>}
     uVRegion {bIsValid=false, max=<0 0>, min=<0 0>}
    }}
   slot {class=ButtonSlot}
  }
  btnRefresh {
   widgetClassPath=&EditorUtilityButton
   parent=HorizontalBox_0
   widget=btnRefresh
   properties {
    toolTipText=Find the selected asset in the content browser
    widgetStyle {
     clickedSlateSound {resourceObject=null}
     disabled {
      bIsDynamicallyLoaded=false
      drawAs=RoundedBox
      imageType=NoImage
      mirroring=NoMirror
      resourceName=None
      resourceObject=null
      tiling=NoTile
      imageSize=<0 0>
      margin=<ltrb 0 0 0 0>
      outlineSettings {
       bUseBrushTransparency=false
       roundingType=FixedRadius
       width=1
       color {colorUseRule=UseColor_ColorTable, specifiedColor=<rgba 0 0 0 0>}
       cornerRadii=<4 4 4 4>
      }
      tintColor {colorUseRule=UseColor_Specified, specifiedColor=<rgba 0.0395 0.0395 0.0395 1>}
      uVRegion {bIsValid=false, max=<0 0>, min=<0 0>}
     }
     disabledForeground {colorUseRule=UseColor_ColorTable, specifiedColor=<rgba 0 0 0 0>}
     hovered {
      bIsDynamicallyLoaded=false
      drawAs=RoundedBox
      imageType=NoImage
      mirroring=NoMirror
      resourceName=None
      resourceObject=null
      tiling=NoTile
      imageSize=<32 32>
      margin=<ltrb 0 0 0 0>
      outlineSettings {
       bUseBrushTransparency=false
       roundingType=FixedRadius
       width=1
       color {colorUseRule=UseColor_ColorTable, specifiedColor=<rgba 0 0 0 0>}
       cornerRadii=<4 4 4 4>
      }
      tintColor {colorUseRule=UseColor_Specified, specifiedColor=<rgba 0.0953 0.0953 0.0953 1>}
      uVRegion {bIsValid=false, max=<0 0>, min=<0 0>}
     }
     hoveredForeground {colorUseRule=UseColor_ColorTable, specifiedColor=<rgba 0 0 0 0>}
     hoveredSlateSound {resourceObject=null}
     normal {
      bIsDynamicallyLoaded=false
      drawAs=RoundedBox
      imageType=NoImage
      mirroring=NoMirror
      resourceName=None
      resourceObject=null
      tiling=NoTile
      imageSize=<32 32>
      margin=<ltrb 0 0 0 0>
      outlineSettings {
       bUseBrushTransparency=false
       roundingType=FixedRadius
       width=1
       color {colorUseRule=UseColor_ColorTable, specifiedColor=<rgba 0 0 0 0>}
       cornerRadii=<4 4 4 4>
      }
      tintColor {colorUseRule=UseColor_Specified, specifiedColor=<rgba 0.0395 0.0395 0.0395 1>}
      uVRegion {bIsValid=false, max=<0 0>, min=<0 0>}
     }
     normalForeground {colorUseRule=UseColor_ColorTable, specifiedColor=<rgba 0 0 0 0>}
     normalPadding=<ltrb 0 2 0 2>
     pressed {
      bIsDynamicallyLoaded=false
      drawAs=RoundedBox
      imageType=NoImage
      mirroring=NoMirror
      resourceName=None
      resourceObject=null
      tiling=NoTile
      imageSize=<32 32>
      margin=<ltrb 0 0 0 0>
      outlineSettings {
       bUseBrushTransparency=false
       roundingType=FixedRadius
       width=1
       color {colorUseRule=UseColor_ColorTable, specifiedColor=<rgba 0 0 0 0>}
       cornerRadii=<4 4 4 4>
      }
      tintColor {colorUseRule=UseColor_Specified, specifiedColor=<rgba 0.0284 0.0284 0.0284 1>}
      uVRegion {bIsValid=false, max=<0 0>, min=<0 0>}
     }
     pressedForeground {colorUseRule=UseColor_ColorTable, specifiedColor=<rgba 0 0 0 0>}
     pressedPadding=<ltrb 0 0 0 0>
     pressedSlateSound {resourceObject=null}
    }
   }
   slot {class=HorizontalBoxSlot, properties {horizontalAlignment=HAlign_Center, verticalAlignment=VAlign_Center, padding=<ltrb 3 0 3 0>}}
  }
  Image {
   widgetClassPath=&Image
   bIsVariable=true
   parent=btnRefresh
   widget=Image
   properties {brush {
     bIsDynamicallyLoaded=false
     drawAs=Image
     imageType=FullColor
     mirroring=NoMirror
     resourceName=None
     resourceObject=/Game/Athena/UI/InGame/Creative/PropertyEditor/Icon_Picker/Textures_V2/64x64/T_UI_IconLibrary_Reset_64
     tiling=NoTile
     imageSize=<18 18>
     margin=<ltrb 0 0 0 0>
     outlineSettings {
      bUseBrushTransparency=false
      roundingType=HalfHeightRadius
      width=0
      color {colorUseRule=UseColor_Specified, specifiedColor=<rgba 0 0 0 0>}
      cornerRadii=<0 0 0 0>
     }
     tintColor {colorUseRule=UseColor_Specified, specifiedColor=<rgba 1 1 1 1>}
     uVRegion {bIsValid=false, max=<0 0>, min=<0 0>}
    }}
   slot {class=ButtonSlot}
  }
 ]
}
